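; Creative Halloween jack-o-lantern prop blueprint that adds a pumpkin destruction effect when the prop dies.
; The death-effects event runs the parent behavior, then spawns the pumpkin Niagara system attached to the static mesh component.
; The effect takes its rotation from the DeathFXRotation variable, keeps a relative offset and auto-destroys.
; The construction script is empty.
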(event EventOnDeathPlayEffects
  (params (Damage :type Float%20%28single-precision%29) (DamageTags :type Gameplay%20Tag%20Container%20Structure%20%28by%20ref%29) (Momentum :type Vector :default "0, 0, 0") (HitInfo :type Hit%20Result%20Structure%20%28by%20ref%29) (InstigatedBy :type Fort%20Pawn%20Object%20Reference) (DamageCauser :type Actor%20Object%20Reference) (EffectContext :type Gameplay%20Effect%20Context%20Handle%20Structure))
  (|CallParentFunction Damage DamageTags Momentum HitInfo InstigatedBy DamageCauser EffectContext)
  (Effects|SpawnFXSystemAttached :FXSystemAsset "/Game/Effects/Fort_Effects/World/Destruction/Halloween/NS_Destruction_Pumpkin.NS_Destruction_Pumpkin" :AttachToComponent (Variables|Debug|GetStaticMeshComponent) :AttachPointName "None" :Rotation (Variables|Default|GetDeathFXRotation) :Scale "1.000000,1.000000,1.000000" :LocationType "KeepRelativeOffset" :bAutoDestroy true :PoolingMethod "None" :bAutoActivate true))

(fn ConstructionScript)
