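; Blueprint for a Halloween jack-o'-lantern pumpkin building prop, placed under the Creative props folder.
; BeginPlay calls the parent implementation and evaluates a cosmetic-events check with nothing wired after it.
; The construction script is empty, so the graphs shown add no behavior of their own beyond what the parent class provides.
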(event EventBeginPlay
  (|CallParentFunction)
  (Networking|CanExecuteCosmeticEvents))

(fn ConstructionScript)
